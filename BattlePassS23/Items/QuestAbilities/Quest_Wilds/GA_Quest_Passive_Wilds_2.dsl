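; Passive quest ability for a Battle Pass S23 Wilds quest that credits the player for distinct big bushes overlapped when a configured gameplay tag is added.
; On each tag add, it gathers overlapping actors by class filter, casts them to Apollo_BigBush, records unused ones and sends a custom quest stat event for each.
; Activation and failed passive activation duplicate this logic, and both start a ConsumedBerry event listener whose handler is empty.
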(event Ability|EventFailedToActivatePassiveAbility
  (params (ActorInfo :type Gameplay%20Ability%20Actor%20Info%20Structure))
  (Ability|Tasks|WaitGameplayTagAdd (Variables|Default|GetTag)
    (:then
      (Ability|Async|WaitGameplayEventToActor (Variables|GAVars|Data|GetOwningPlayerPawn) "(TagName=\"Athena.event.gameplay.ConsumedBerry\")"
        (:EventReceived
          _)))
    (:Added
      (for _array_element (Collision|GetOverlappingActors (Variables|GAVars|Data|GetOwningPlayerPawn) (Variables|Default|GetClass%20Filter) :!class /Script/Engine.Actor)
        (bind _asapollo_big_bush (Utilities|Casting|CastToApollo_BigBush _array_element))
        (if (not (Utilities|Array|ContainsItem (Variables|Default|GetUsed%20Bushes) _asapollo_big_bush))
          (Utilities|Array|AddUnique (Variables|Default|GetUsed%20Bushes) _asapollo_big_bush)
          (Quest|SendComplexCustomStatEvent (Variables|GAVars|Data|GetOwningPlayerQuestManager) 0 (Variables|Default|GetAdditional%20Source%20Tags) "(GameplayTags=)" 1))))))

(event Ability|EventActivateAbility
  (|CallParentFunction)
  (Ability|Tasks|WaitGameplayTagAdd (Variables|Default|GetTag)
    (:then
      (Ability|Async|WaitGameplayEventToActor (Variables|GAVars|Data|GetOwningPlayerPawn) "(TagName=\"Athena.event.gameplay.ConsumedBerry\")"
        (:EventReceived
          _)))
    (:Added
      (for _array_element (Collision|GetOverlappingActors (Variables|GAVars|Data|GetOwningPlayerPawn) (Variables|Default|GetClass%20Filter) :!class /Script/Engine.Actor)
        (bind _asapollo_big_bush (Utilities|Casting|CastToApollo_BigBush _array_element))
        (if (not (Utilities|Array|ContainsItem (Variables|Default|GetUsed%20Bushes) _asapollo_big_bush))
          (Utilities|Array|AddUnique (Variables|Default|GetUsed%20Bushes) _asapollo_big_bush)
          (Quest|SendComplexCustomStatEvent (Variables|GAVars|Data|GetOwningPlayerQuestManager) 0 (Variables|Default|GetAdditional%20Source%20Tags) "(GameplayTags=)" 1))))))
